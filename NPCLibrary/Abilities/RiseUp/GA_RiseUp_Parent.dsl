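; Parent gameplay ability for an NPC "rise up" entrance, with overridable pre-montage and during-montage hooks.
; Activation caches the AI pawn and speech bubble, then starts the montage, which leads to an optional delayed intro speech.
; Optional starting knockback uses target selection to launch pawns away, push vehicles through an impact call, and apply a gameplay effect container.
; Ending the ability plays optional end speech and unhides the NPC if hidden.
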(event Ability|EventActivateAbility
  (bind _outfortplayerpawn (CallFunction|GetAvatarasValidFortPlayerPawn)
    (:ValidFortPlayerPawn
      (bind _output_get (Variables|InternalData|SetAIPawn _outfortplayerpawn))
      (Variables|InternalData|SetSpeechBubbleComponent (Actor|GetComponentbyClass _output_get "/Script/FortniteGame.FortPawnComponent_SpeechBubble"))
      (Utilities|FlowControl|Delay 0.1)
      (CallFunction|HandlePreMontage)
      (if (Variables|Default|GetKnockbackPlayersAtStart)
        (CallFunction|KnockbackPlayers)))
    (:AvatarCastFailed
      (Ability|EndAbility))))

(event Ability|EventOnEndAbility
  (params (bWasCancelled :type Boolean))
  (bind _aipawn (Variables|InternalData|GetAIPawn))
  (CallFunction|PlayEndRiseUpSpeech)
  (if (Variables|/Script/Engine.Actor|GetbHidden _aipawn)
    (Rendering|SetActorHiddenInGame _aipawn false)))

(event Ability|EventFailedToActivatePassiveAbility
  (params (ActorInfo :type Gameplay%20Ability%20Actor%20Info%20Structure)))

(event Custom|HandlePreMontage
  (CallFunction|StartMontage))

(event Custom|HandleDuringMontage)

(event Custom|StartMontage
  (Ability|Tasks|PlayMontageInfoAndWaitForTrigger "None" (Variables|Animation|GetMontageInfo)
    (:then
      (CallFunction|HandleDuringMontage)
      (Utilities|FlowControl|Delay 1.0)
      (CallFunction|PlayIntroSpeech))
    (:Cancelled
      _)
    (:Completed
      _))
  (Ability|EndAbility))

(event Custom|PlayIntroSpeech
  (bind _speechbubblecomponent (Variables|InternalData|GetSpeechBubbleComponent))
  (if (Variables|Default|GetTriggerIntroSpeech)
    (Utilities|IsValid _speechbubblecomponent
      (:Is%20Valid
        (SpeechBubble|StartSpeech _speechbubblecomponent (Variables|Default|GetTagQuery_IntroSpeech))))))

(event Custom|PlayEndRiseUpSpeech
  (bind _speechbubblecomponent (Variables|InternalData|GetSpeechBubbleComponent))
  (if (Variables|Default|GetTriggerEndSpeech)
    (Utilities|IsValid _speechbubblecomponent
      (:Is%20Valid
        (SpeechBubble|StartSpeech _speechbubblecomponent (Variables|Default|GetTagQuery_EndSpeech))))))

(event Custom|KnockbackPlayers
  (bind (_asynctaskproxy _targetdata _applicationtag) (Ability|Tasks|DoTargetSelection "None" "(TagName=\"Targeting.ApplicationTag.0\")" true)
    (:Targeted
      (for _array_element (Ability|TargetData|GetAllActorsfromTargetData _targetdata)
        (bind _asfort_pawn (Utilities|Casting|CastToFortPawn _array_element)
          (:then
            (CallFunction|KnockbackPawn _asfort_pawn))
          (:CastFailed
            (bind _asfort_athena_vehicle (Utilities|Casting|CastToFortAthenaVehicle _array_element))
            (CallFunction|KnockbackVehicle _asfort_athena_vehicle))))
      (Ability|ApplyGameplayEffectContainer _targetdata _applicationtag))))

(fn KnockbackPawn
  (params (FortPawn :type Fort%20Pawn%20Object%20Reference))
  (bind (_returnvalue_x _returnvalue_y) (Math|Vector|BreakVector (Math|Vector|Normalize (- (Transformation|GetActorLocation FortPawn) (Transformation|GetActorLocation (Variables|InternalData|GetAIPawn))))))
  (Character|LaunchCharacter FortPawn (Math|Vector|MakeVector (* _returnvalue_x (ScalableFloat|GetValueatLevel (Variables|Default|GetKnockbackPawnForwardVelocity))) (* _returnvalue_y (ScalableFloat|GetValueatLevel (Variables|Default|GetKnockbackPawnForwardVelocity))) (ScalableFloat|GetValueatLevel (Variables|Default|GetKnockbackPawnHeightVelocity))) true true))

(fn KnockbackVehicle
  (params (FortAthenaVehicle :type Fort%20Athena%20Vehicle%20Object%20Reference))
  (bind (_returnvalue_x _returnvalue_y) (Math|Vector|BreakVector (Math|Vector|Normalize (- (Transformation|GetActorLocation FortAthenaVehicle) (Transformation|GetActorLocation (Variables|InternalData|GetAIPawn))))))
  (Audio|BroadcastAbilityImpactatLocation 0 FortAthenaVehicle (Math|Vector|MakeVector (* _returnvalue_x (ScalableFloat|GetValueatLevel (Variables|Default|GetKnockbackVehicleForwardVelocity))) (* _returnvalue_y (ScalableFloat|GetValueatLevel (Variables|Default|GetKnockbackVehicleForwardVelocity))) (ScalableFloat|GetValueatLevel (Variables|Default|GetKnockbackVehicleHeightVelocity))) 777.0 "None" false))
